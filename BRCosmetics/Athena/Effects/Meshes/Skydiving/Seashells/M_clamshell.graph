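# Lit clamshell material for the seashell skydiving effect meshes, usable on mesh particles and Niagara systems.
# Sine-warped UVs create a procedural pattern that blends two dark color parameters and drives a narrow scalar range.
# A sharpened Fresnel rim blends in a time-of-day tint derived from ColorA.
# A camera-distance offset collapses vertices toward the object pivot as the camera nears, controlled by DistanceShrink and zeroed on mobile.
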Material /BRCosmetics/Athena/Effects/Meshes/Skydiving/Seashells/M_clamshell
graph {
 properties {
  bUsedWithMeshParticles=true
  bUsedWithNiagaraMeshParticles=true
  bUsedWithNiagaraRibbons=true
  bUsedWithNiagaraSprites=true
  pixelDepthOffsetMode=PDOM_Legacy
  usedShadingModels=Default Lit
 }
 nodes [
  n1 = TextureCoordinate()
  n2 = MaterialFunctionCall(Float2=Sine(None=Multiply(A=@n1) {constB=10})) {materialFunction=&BreakOutFloat2Components}
  n3 = Add(A=Multiply(A=@n2.G) {constB=0.5}) {constB=0.5}
  Add(A=Multiply(A=OneMinus(None=@n3)) {constB=0.01}) {constB=0.1}
  n4 = VectorParameter() {parameterName=ColorA, defaultValue=<rgba 0.0125 0.011 0.0131 1>}
  n5 = Add(A=Add(A=Multiply(A=@n2.R) {constB=0.25}) {constB=0.25}, B=@n3)
  n6 = Multiply(A=@n4.RGB, B=@n4.RGB) {constB=0.4}
  n7 = Fresnel()
  n8 = Multiply(A=@n7, B=@n7)
  n9 = Multiply(A=@n8) {constB=30}
  LinearInterpolate(
   A=LinearInterpolate(A=@n4.RGB, B=VectorParameter() {parameterName=ColorB, defaultValue=<rgba 0.0111 0.0073 0.008 1>}.RGB, Alpha=Clamp(None=@n5))
   B=MaterialFunctionCall(Color_Morning=@n4.RGB, Color_Day=@n6, Color_Evening=Multiply(A=@n6) {constB=0.75}, Color_Night=Multiply(A=@n6) {constB=0.25}) {materialFunction=&Color_Time_of_Day}.Result
   Alpha=Clamp(None=Multiply(A=@n9, B=@n9))
  )
  Multiply(A=@n8, B=@n8) {constB=10}
  n10 = MaterialFunctionCall() {materialFunction=&ObjectPivotPoint}
  n11 = WorldPosition()
  n12 = Divide(A=Distance(A=CameraPositionWS(), B=@n10.Object Pivot Location), B=ScalarParameter() {defaultValue=100, parameterName=DistanceShrink})
  ShadingPathSwitch(Default=LinearInterpolate(A=Subtract(A=@n10.Object Pivot Location, B=@n11.XYZ), Alpha=Saturate(None=@n12)) {constB=0}, Mobile=Constant()) {desc=Migrated from FeatureLevelSwitch}
 ]
}
